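# Control Rig for the SirWolf medium male player head that applies a corrective to the dyn_hair_05 bone as the head moves.
# A spherical pose reader measures the head bone relative to neck_02, using asymmetric active and falloff regions.
# Its output weights a Modify Transforms step that shrinks the hair bone along local X with zero rotation offset.
# The hierarchy also holds face bones and many curves, which the graph leaves untouched.
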ControlRigBlueprint /Game/Characters/Player/Male/Medium/Heads/M_MED_SirWolf_Head/Meshes/M_MED_SirWolf_Head_CtrlRig
graphs [
 RigVMFunctionLibrary {links [], nodes []}
 RigVMModel {nodes [
   RigUnit_BeginExecution = &RigUnit_BeginExecution()
   n1 = &RigUnit_SphericalPoseReader(
    ActiveRegionScaleFactors=(PositiveWidth=0.25,NegativeWidth=2,PositiveHeight=0.25,NegativeHeight=0.25)
    ActiveRegionSize=0.1
    Debug=(bDrawDebug=false,DebugScale=25,DebugSegments=20,DebugThickness=0.25)
    DriverAxis=(X=0,Y=-1,Z=0)
    DriverItem=(Type=Bone,Name="head")
    FalloffRegionScaleFactors=(PositiveWidth=0.75,NegativeWidth=2,PositiveHeight=2,NegativeHeight=2)
    FalloffSize=0.2
    OptionalParentItem=(Type=Bone,Name="neck_02")
    RotationOffset=(X=-90,Y=0,Z=-10)
    ExecutePin=@RigUnit_BeginExecution.ExecutePin
   )
   > &RigUnit_ModifyTransforms(
    ItemToModify=((Item=(Type=Bone,Name="dyn_hair_05"),Transform=(Rotation=(X=0,Y=0,Z=0,W=1),Translation=(X=0,Y=0,Z=0),Scale3D=(X=0.75,Y=1,Z=1))))
    ItemToModify.0.Transform.Rotation=&RigVMFunction_MathQuaternionFromEuler(Euler=(X=0,Y=0,Z=0)).Result
    Weight=@n1.OutputParam
   )
  ]}
]
hierarchy [
 root {loc=<0 0 0>, rot=<0 0 0 1>}
 pelvis {parent=root, loc=<0 -0.7071 93.1191>, rot=<0.0127 0.707 -0.0127 -0.707>}
 spine_01 {parent=pelvis, loc=<6.7494 -0.6388 0>, rot=<0 0 0.034 -0.9994>}
 spine_02 {parent=spine_01, loc=<7.2518 -0.035 0>, rot=<0 0 0.0238 -0.9997>}
 spine_03 {parent=spine_02, loc=<6.0104 0 0>, rot=<0 0 0.0149 -0.9999>}
 spine_04 {parent=spine_03, loc=<8.1024 0 0>, rot=<0 0 -0.0132 -0.9999>}
 spine_05 {parent=spine_04, loc=<10.0146 0 0>, rot=<0 0 -0.0015 -1>}
 neck_01 {parent=spine_05, loc=<11.3138 0.1529 0>, rot=<0 0 -0.2003 -0.9797>}
 neck_02 {parent=neck_01, loc=<4.5627 0 0>, rot=<0 0 0.0504 -0.9987>}
 head {parent=neck_02, loc=<5.3914 0 0>, rot=<0 0 0.115 -0.9934>}
 faceAttach {parent=head, loc=<-4.146 -0.8545 0>, rot=<0.0029 -0.7071 -0.0029 -0.7071>}
 C_brow_mid {parent=faceAttach, loc=<0 9.8905 15.2152>, rot=<0 0 0 -1>}
 C_jaw {parent=faceAttach, loc=<-0.0134 1.6187 11.5059>, rot=<0 0 0 1>}
 C_lip_lower_mid {parent=C_jaw, loc=<0.0223 16.903 -10.2943>, rot=<0.2651 0 0 -0.9642>}
 L_lip_lower_outer {parent=C_jaw, loc=<2.3841 12.6261 -7.7123>, rot=<0.0256 -0.0404 0.0049 -0.9988>}
 R_lip_lower_outer {parent=C_jaw, loc=<-1.5712 13.1898 -7.8023>, rot=<0.0256 0.04 -0.0049 -0.9989>}
 teeth_lower {parent=C_jaw, loc=<0.0134 8.0181 -5.6575>, rot=<0.2775 0 0 -0.9607>}
 tongue {parent=C_jaw, loc=<0.0141 6.2817 -3.8712>, rot=<0.2024 0 0 -0.9793>}
 C_lip_upper_mid {parent=faceAttach, loc=<0.0022 19.9362 3.4006>, rot=<0.2225 0.0013 -0.0005 -0.9749>}
 C_nose_bridge {parent=faceAttach, loc=<0 21.223 6.3909>, rot=<0.1395 0 0 -0.9902>}
 L_brow_mid {parent=faceAttach, loc=<2.1723 9.6608 15.1421>, rot=<-0.0009 -0.0041 0.0772 -0.997>}
 L_brow_outer {parent=faceAttach, loc=<5.1605 8.4981 14.046>, rot=<-0.0342 -0.0353 0.3138 -0.9482>}
 L_cheek_inner {parent=faceAttach, loc=<4.6912 7.1284 9.0618>, rot=<-0.0706 -0.2697 0.3184 -0.906>}
 L_eye {parent=faceAttach, loc=<2.5429 7.8209 11.794>, rot=<0 0 0.9998 0.0175>}
 L_eye_lid_lower_mid {parent=L_eye, loc=<0 0 0>, rot=<0.0329 0 0 -0.9995>}
 L_eye_lid_upper_mid {parent=L_eye, loc=<0 0 0>, rot=<0.0523 0 0 -0.9986>}
 L_lip_corner {parent=faceAttach, loc=<3.8112 9.5854 6.8087>, rot=<0.2331 0.0111 -0.0124 -0.9723>}
 L_lip_upper_outer {parent=faceAttach, loc=<2.3228 14.5151 4.2496>, rot=<0.0755 0.0153 -0.0129 -0.9969>}
 R_brow_mid {parent=faceAttach, loc=<-2.1722 9.6603 15.1423>, rot=<-0.0009 0.0041 -0.0772 -0.997>}
 R_brow_outer {parent=faceAttach, loc=<-5.1605 8.4981 14.046>, rot=<-0.0342 0.0353 -0.3138 -0.9482>}
 R_cheek_inner {parent=faceAttach, loc=<-4.6912 7.1284 9.0618>, rot=<-0.0706 0.2697 -0.3184 -0.906>}
 R_eye {parent=faceAttach, loc=<-2.543 7.8206 11.794>, rot=<0 0 0.9998 -0.0175>}
 R_eye_lid_lower_mid {parent=R_eye, loc=<0 0 0>, rot=<0.0329 0 0 -0.9995>}
 R_eye_lid_upper_mid {parent=R_eye, loc=<0 0 0>, rot=<0.0523 0 0 -0.9986>}
 R_lip_corner {parent=faceAttach, loc=<-3.8112 9.5854 6.8087>, rot=<0.2331 -0.0111 0.0124 -0.9723>}
 R_lip_upper_outer {parent=faceAttach, loc=<-1.4957 15.0109 4.3672>, rot=<0.0755 -0.0153 0.0129 -0.9969>}
 dyn_hair_01 {parent=head, loc=<11.9517 1.7862 -5.9836>, rot=<0.002 -0.4951 -0.0036 -0.8688>}
 dyn_hair_02 {parent=head, loc=<11.9514 1.7862 5.9836>, rot=<-0.4951 -0.002 0.8688 -0.0036>}
 dyn_hair_03 {parent=head, loc=<2.6897 -0.1968 -4.8779>, rot=<0.0029 -0.7083 -0.0029 -0.7059>}
 dyn_hair_04 {parent=head, loc=<2.6895 -0.1968 4.8779>, rot=<-0.0029 0.7083 -0.0029 -0.7059>}
 dyn_hair_05 {parent=head, loc=<2.7143 -3.173 0>, rot=<0.5029 -0.4988 0.4971 -0.5012>}
 look_left_pose {type=Curve, loc=<0 0 0>, rot=<0 0 0 1>}
 look_right_pose {type=Curve, loc=<0 0 0>, rot=<0 0 0 1>}
 look_up_pose {type=Curve, loc=<0 0 0>, rot=<0 0 0 1>}
 look_down_pose {type=Curve, loc=<0 0 0>, rot=<0 0 0 1>}
 L_blink_pose {type=Curve, loc=<0 0 0>, rot=<0 0 0 1>}
 R_blink_pose {type=Curve, loc=<0 0 0>, rot=<0 0 0 1>}
 L_wide_pose {type=Curve, loc=<0 0 0>, rot=<0 0 0 1>}
 R_wide_pose {type=Curve, loc=<0 0 0>, rot=<0 0 0 1>}
 R_squint_inner_pose {type=Curve, loc=<0 0 0>, rot=<0 0 0 1>}
 R_squeeze_pose {type=Curve, loc=<0 0 0>, rot=<0 0 0 1>}
 L_squint_inner_pose {type=Curve, loc=<0 0 0>, rot=<0 0 0 1>}
 L_squeeze_pose {type=Curve, loc=<0 0 0>, rot=<0 0 0 1>}
 L_brow_up_pose {type=Curve, loc=<0 0 0>, rot=<0 0 0 1>}
 L_brow_down_pose {type=Curve, loc=<0 0 0>, rot=<0 0 0 1>}
 R_brow_up_pose {type=Curve, loc=<0 0 0>, rot=<0 0 0 1>}
 R_brow_down_pose {type=Curve, loc=<0 0 0>, rot=<0 0 0 1>}
 C_glabella_up_pose {type=Curve, loc=<0 0 0>, rot=<0 0 0 1>}
 C_glabella_down_pose {type=Curve, loc=<0 0 0>, rot=<0 0 0 1>}
 L_brow_squeeze_in_pose {type=Curve, loc=<0 0 0>, rot=<0 0 0 1>}
 R_brow_squeeze_in_pose {type=Curve, loc=<0 0 0>, rot=<0 0 0 1>}
 L_upper_lip_raiser_pose {type=Curve, loc=<0 0 0>, rot=<0 0 0 1>}
 R_upper_lip_raiser_pose {type=Curve, loc=<0 0 0>, rot=<0 0 0 1>}
 L_upper_lip_lower_pose {type=Curve, loc=<0 0 0>, rot=<0 0 0 1>}
 R_upper_lip_lower_pose {type=Curve, loc=<0 0 0>, rot=<0 0 0 1>}
 L_lower_lip_up_pose {type=Curve, loc=<0 0 0>, rot=<0 0 0 1>}
 R_lower_lip_up_pose {type=Curve, loc=<0 0 0>, rot=<0 0 0 1>}
 L_lower_lip_down_pose {type=Curve, loc=<0 0 0>, rot=<0 0 0 1>}
 R_lower_lip_down_pose {type=Curve, loc=<0 0 0>, rot=<0 0 0 1>}
 L_lip_corner_narrow_pose {type=Curve, loc=<0 0 0>, rot=<0 0 0 1>}
 L_lip_corner_wide_pose {type=Curve, loc=<0 0 0>, rot=<0 0 0 1>}
 R_lip_corner_narrow_pose {type=Curve, loc=<0 0 0>, rot=<0 0 0 1>}
 R_lip_corner_wide_pose {type=Curve, loc=<0 0 0>, rot=<0 0 0 1>}
 L_smile_pose {type=Curve, loc=<0 0 0>, rot=<0 0 0 1>}
 R_smile_pose {type=Curve, loc=<0 0 0>, rot=<0 0 0 1>}
 L_frown_pose {type=Curve, loc=<0 0 0>, rot=<0 0 0 1>}
 R_frown_pose {type=Curve, loc=<0 0 0>, rot=<0 0 0 1>}
 phoneme_oo_pose {type=Curve, loc=<0 0 0>, rot=<0 0 0 1>}
 phoneme_ch_pose {type=Curve, loc=<0 0 0>, rot=<0 0 0 1>}
 phoneme_mbp_pose {type=Curve, loc=<0 0 0>, rot=<0 0 0 1>}
 phoneme_fv_pose {type=Curve, loc=<0 0 0>, rot=<0 0 0 1>}
 mouth_up_pose {type=Curve, loc=<0 0 0>, rot=<0 0 0 1>}
 mouth_down_pose {type=Curve, loc=<0 0 0>, rot=<0 0 0 1>}
 C_nose_wrinkler_pose {type=Curve, loc=<0 0 0>, rot=<0 0 0 1>}
 L_cheek_up_pose {type=Curve, loc=<0 0 0>, rot=<0 0 0 1>}
 R_cheek_up_pose {type=Curve, loc=<0 0 0>, rot=<0 0 0 1>}
 jaw_open_pose {type=Curve, loc=<0 0 0>, rot=<0 0 0 1>}
 jaw_right_pose {type=Curve, loc=<0 0 0>, rot=<0 0 0 1>}
 jaw_left_pose {type=Curve, loc=<0 0 0>, rot=<0 0 0 1>}
 eye_convergence_pose {type=Curve, loc=<0 0 0>, rot=<0 0 0 1>}
 L_upperLidDown_pose {type=Curve, loc=<0 0 0>, rot=<0 0 0 1>}
 R_upperLidDown_pose {type=Curve, loc=<0 0 0>, rot=<0 0 0 1>}
 jaw_up_pose {type=Curve, loc=<0 0 0>, rot=<0 0 0 1>}
 jaw_forward_pose {type=Curve, loc=<0 0 0>, rot=<0 0 0 1>}
 jaw_back_pose {type=Curve, loc=<0 0 0>, rot=<0 0 0 1>}
 tongue_forward_pose {type=Curve, loc=<0 0 0>, rot=<0 0 0 1>}
 tongue_back_pose {type=Curve, loc=<0 0 0>, rot=<0 0 0 1>}
 tongue_up_pose {type=Curve, loc=<0 0 0>, rot=<0 0 0 1>}
 tongue_down_pose {type=Curve, loc=<0 0 0>, rot=<0 0 0 1>}
 base_pose {type=Curve, loc=<0 0 0>, rot=<0 0 0 1>}
 Hair_Scarf {type=Curve, loc=<0 0 0>, rot=<0 0 0 1>}
 Hair_Hat {type=Curve, loc=<0 0 0>, rot=<0 0 0 1>}
 Hair_Helmet {type=Curve, loc=<0 0 0>, rot=<0 0 0 1>}
 Hair_Glasses {type=Curve, loc=<0 0 0>, rot=<0 0 0 1>}
 Hair_Cap {type=Curve, loc=<0 0 0>, rot=<0 0 0 1>}
 Hair_Headsets {type=Curve, loc=<0 0 0>, rot=<0 0 0 1>}
 Hair_Hood {type=Curve, loc=<0 0 0>, rot=<0 0 0 1>}
 Face_Mask {type=Curve, loc=<0 0 0>, rot=<0 0 0 1>}
 HairStrands_Hat {type=Curve, loc=<0 0 0>, rot=<0 0 0 1>}
 Hat_Hair {type=Curve, loc=<0 0 0>, rot=<0 0 0 1>}
 Helmet_Hair {type=Curve, loc=<0 0 0>, rot=<0 0 0 1>}
 Hair_Muff {type=Curve, loc=<0 0 0>, rot=<0 0 0 1>}
 Hair_SpaceSuit {type=Curve, loc=<0 0 0>, rot=<0 0 0 1>}
 Hair_Hide {type=Curve, loc=<0 0 0>, rot=<0 0 0 1>}
 Head_SpaceSuit {type=Curve, loc=<0 0 0>, rot=<0 0 0 1>}
 Hair_Null {type=Curve, loc=<0 0 0>, rot=<0 0 0 1>}
 DisableFaceOverride {type=Curve, loc=<0 0 0>, rot=<0 0 0 1>}
 LiveLink {type=Curve, loc=<0 0 0>, rot=<0 0 0 1>}
 Hat_MESH {type=Curve, loc=<0 0 0>, rot=<0 0 0 1>}
 Null_MESH {type=Curve, loc=<0 0 0>, rot=<0 0 0 1>}
 Pose_0 {type=Curve, loc=<0 0 0>, rot=<0 0 0 1>}
 Hair_Mask {type=Curve, loc=<0 0 0>, rot=<0 0 0 1>}
 eye_convergence {type=Curve, loc=<0 0 0>, rot=<0 0 0 1>}
 headTransform {type=Curve, loc=<0 0 0>, rot=<0 0 0 1>}
 Pose_1 {type=Curve, loc=<0 0 0>, rot=<0 0 0 1>}
 Pose_2 {type=Curve, loc=<0 0 0>, rot=<0 0 0 1>}
 Pose_3 {type=Curve, loc=<0 0 0>, rot=<0 0 0 1>}
 Pose_4 {type=Curve, loc=<0 0 0>, rot=<0 0 0 1>}
 Pose_5 {type=Curve, loc=<0 0 0>, rot=<0 0 0 1>}
 Pose_6 {type=Curve, loc=<0 0 0>, rot=<0 0 0 1>}
 Pose_7 {type=Curve, loc=<0 0 0>, rot=<0 0 0 1>}
 Pose_8 {type=Curve, loc=<0 0 0>, rot=<0 0 0 1>}
 Pose_9 {type=Curve, loc=<0 0 0>, rot=<0 0 0 1>}
 Pose_10 {type=Curve, loc=<0 0 0>, rot=<0 0 0 1>}
 Pose_11 {type=Curve, loc=<0 0 0>, rot=<0 0 0 1>}
 Pose_12 {type=Curve, loc=<0 0 0>, rot=<0 0 0 1>}
 Pose_13 {type=Curve, loc=<0 0 0>, rot=<0 0 0 1>}
 Pose_14 {type=Curve, loc=<0 0 0>, rot=<0 0 0 1>}
 Pose_15 {type=Curve, loc=<0 0 0>, rot=<0 0 0 1>}
 Pose_16 {type=Curve, loc=<0 0 0>, rot=<0 0 0 1>}
 Pose_17 {type=Curve, loc=<0 0 0>, rot=<0 0 0 1>}
 Pose_18 {type=Curve, loc=<0 0 0>, rot=<0 0 0 1>}
 Pose_19 {type=Curve, loc=<0 0 0>, rot=<0 0 0 1>}
 Pose_20 {type=Curve, loc=<0 0 0>, rot=<0 0 0 1>}
 Pose_21 {type=Curve, loc=<0 0 0>, rot=<0 0 0 1>}
 Pose_22 {type=Curve, loc=<0 0 0>, rot=<0 0 0 1>}
 Pose_23 {type=Curve, loc=<0 0 0>, rot=<0 0 0 1>}
 Pose_24 {type=Curve, loc=<0 0 0>, rot=<0 0 0 1>}
 Pose_25 {type=Curve, loc=<0 0 0>, rot=<0 0 0 1>}
 Pose_26 {type=Curve, loc=<0 0 0>, rot=<0 0 0 1>}
 Pose_27 {type=Curve, loc=<0 0 0>, rot=<0 0 0 1>}
 Pose_28 {type=Curve, loc=<0 0 0>, rot=<0 0 0 1>}
 Pose_29 {type=Curve, loc=<0 0 0>, rot=<0 0 0 1>}
 Pose_30 {type=Curve, loc=<0 0 0>, rot=<0 0 0 1>}
 Pose_31 {type=Curve, loc=<0 0 0>, rot=<0 0 0 1>}
 Pose_32 {type=Curve, loc=<0 0 0>, rot=<0 0 0 1>}
 Pose_33 {type=Curve, loc=<0 0 0>, rot=<0 0 0 1>}
 Pose_34 {type=Curve, loc=<0 0 0>, rot=<0 0 0 1>}
 Pose_35 {type=Curve, loc=<0 0 0>, rot=<0 0 0 1>}
 Pose_36 {type=Curve, loc=<0 0 0>, rot=<0 0 0 1>}
 Pose_37 {type=Curve, loc=<0 0 0>, rot=<0 0 0 1>}
 Pose_38 {type=Curve, loc=<0 0 0>, rot=<0 0 0 1>}
 Pose_39 {type=Curve, loc=<0 0 0>, rot=<0 0 0 1>}
 Pose_40 {type=Curve, loc=<0 0 0>, rot=<0 0 0 1>}
 Pose_41 {type=Curve, loc=<0 0 0>, rot=<0 0 0 1>}
 Pose_42 {type=Curve, loc=<0 0 0>, rot=<0 0 0 1>}
 Pose_43 {type=Curve, loc=<0 0 0>, rot=<0 0 0 1>}
 Pose_44 {type=Curve, loc=<0 0 0>, rot=<0 0 0 1>}
 Pose_45 {type=Curve, loc=<0 0 0>, rot=<0 0 0 1>}
 Pose_46 {type=Curve, loc=<0 0 0>, rot=<0 0 0 1>}
 Pose_47 {type=Curve, loc=<0 0 0>, rot=<0 0 0 1>}
 Pose_48 {type=Curve, loc=<0 0 0>, rot=<0 0 0 1>}
 Pose_49 {type=Curve, loc=<0 0 0>, rot=<0 0 0 1>}
 Pose_50 {type=Curve, loc=<0 0 0>, rot=<0 0 0 1>}
 Pose_51 {type=Curve, loc=<0 0 0>, rot=<0 0 0 1>}
 Pose_52 {type=Curve, loc=<0 0 0>, rot=<0 0 0 1>}
 Pose_53 {type=Curve, loc=<0 0 0>, rot=<0 0 0 1>}
 Pose_54 {type=Curve, loc=<0 0 0>, rot=<0 0 0 1>}
 Pose_55 {type=Curve, loc=<0 0 0>, rot=<0 0 0 1>}
 Pose_56 {type=Curve, loc=<0 0 0>, rot=<0 0 0 1>}
 Pose_57 {type=Curve, loc=<0 0 0>, rot=<0 0 0 1>}
 Pose_58 {type=Curve, loc=<0 0 0>, rot=<0 0 0 1>}
 Pose_59 {type=Curve, loc=<0 0 0>, rot=<0 0 0 1>}
 Pose_60 {type=Curve, loc=<0 0 0>, rot=<0 0 0 1>}
 Pose_61 {type=Curve, loc=<0 0 0>, rot=<0 0 0 1>}
 Pose_62 {type=Curve, loc=<0 0 0>, rot=<0 0 0 1>}
 Pose_63 {type=Curve, loc=<0 0 0>, rot=<0 0 0 1>}
 Pose_64 {type=Curve, loc=<0 0 0>, rot=<0 0 0 1>}
 Pose_65 {type=Curve, loc=<0 0 0>, rot=<0 0 0 1>}
 Pose_66 {type=Curve, loc=<0 0 0>, rot=<0 0 0 1>}
 Pose_67 {type=Curve, loc=<0 0 0>, rot=<0 0 0 1>}
 Pose_68 {type=Curve, loc=<0 0 0>, rot=<0 0 0 1>}
 Pose_69 {type=Curve, loc=<0 0 0>, rot=<0 0 0 1>}
 Pose_70 {type=Curve, loc=<0 0 0>, rot=<0 0 0 1>}
 Pose_71 {type=Curve, loc=<0 0 0>, rot=<0 0 0 1>}
 Pose_72 {type=Curve, loc=<0 0 0>, rot=<0 0 0 1>}
 Pose_73 {type=Curve, loc=<0 0 0>, rot=<0 0 0 1>}
 Pose_74 {type=Curve, loc=<0 0 0>, rot=<0 0 0 1>}
 Pose_75 {type=Curve, loc=<0 0 0>, rot=<0 0 0 1>}
 Pose_76 {type=Curve, loc=<0 0 0>, rot=<0 0 0 1>}
 Pose_77 {type=Curve, loc=<0 0 0>, rot=<0 0 0 1>}
 Pose_78 {type=Curve, loc=<0 0 0>, rot=<0 0 0 1>}
 Pose_79 {type=Curve, loc=<0 0 0>, rot=<0 0 0 1>}
 Pose_80 {type=Curve, loc=<0 0 0>, rot=<0 0 0 1>}
 Pose_81 {type=Curve, loc=<0 0 0>, rot=<0 0 0 1>}
 Pose_82 {type=Curve, loc=<0 0 0>, rot=<0 0 0 1>}
 Pose_83 {type=Curve, loc=<0 0 0>, rot=<0 0 0 1>}
 Pose_84 {type=Curve, loc=<0 0 0>, rot=<0 0 0 1>}
 Pose_85 {type=Curve, loc=<0 0 0>, rot=<0 0 0 1>}
 Pose_86 {type=Curve, loc=<0 0 0>, rot=<0 0 0 1>}
 Pose_87 {type=Curve, loc=<0 0 0>, rot=<0 0 0 1>}
 Pose_88 {type=Curve, loc=<0 0 0>, rot=<0 0 0 1>}
 Pose_89 {type=Curve, loc=<0 0 0>, rot=<0 0 0 1>}
 Pose_90 {type=Curve, loc=<0 0 0>, rot=<0 0 0 1>}
 Pose_91 {type=Curve, loc=<0 0 0>, rot=<0 0 0 1>}
 Pose_92 {type=Curve, loc=<0 0 0>, rot=<0 0 0 1>}
 Pose_93 {type=Curve, loc=<0 0 0>, rot=<0 0 0 1>}
 Pose_94 {type=Curve, loc=<0 0 0>, rot=<0 0 0 1>}
 Pose_95 {type=Curve, loc=<0 0 0>, rot=<0 0 0 1>}
 Pose_96 {type=Curve, loc=<0 0 0>, rot=<0 0 0 1>}
 Pose_97 {type=Curve, loc=<0 0 0>, rot=<0 0 0 1>}
 Pose_98 {type=Curve, loc=<0 0 0>, rot=<0 0 0 1>}
 Pose_99 {type=Curve, loc=<0 0 0>, rot=<0 0 0 1>}
 Pose_100 {type=Curve, loc=<0 0 0>, rot=<0 0 0 1>}
 Pose_101 {type=Curve, loc=<0 0 0>, rot=<0 0 0 1>}
 Pose_102 {type=Curve, loc=<0 0 0>, rot=<0 0 0 1>}
 Pose_103 {type=Curve, loc=<0 0 0>, rot=<0 0 0 1>}
 Pose_104 {type=Curve, loc=<0 0 0>, rot=<0 0 0 1>}
 Pose_105 {type=Curve, loc=<0 0 0>, rot=<0 0 0 1>}
 Pose_106 {type=Curve, loc=<0 0 0>, rot=<0 0 0 1>}
 Pose_107 {type=Curve, loc=<0 0 0>, rot=<0 0 0 1>}
 Pose_108 {type=Curve, loc=<0 0 0>, rot=<0 0 0 1>}
 Pose_109 {type=Curve, loc=<0 0 0>, rot=<0 0 0 1>}
 Pose_110 {type=Curve, loc=<0 0 0>, rot=<0 0 0 1>}
 Pose_111 {type=Curve, loc=<0 0 0>, rot=<0 0 0 1>}
 Pose_112 {type=Curve, loc=<0 0 0>, rot=<0 0 0 1>}
 Pose_113 {type=Curve, loc=<0 0 0>, rot=<0 0 0 1>}
 Pose_114 {type=Curve, loc=<0 0 0>, rot=<0 0 0 1>}
 Pose_115 {type=Curve, loc=<0 0 0>, rot=<0 0 0 1>}
 Pose_116 {type=Curve, loc=<0 0 0>, rot=<0 0 0 1>}
 Pose_117 {type=Curve, loc=<0 0 0>, rot=<0 0 0 1>}
 Pose_118 {type=Curve, loc=<0 0 0>, rot=<0 0 0 1>}
 Pose_119 {type=Curve, loc=<0 0 0>, rot=<0 0 0 1>}
 Pose_120 {type=Curve, loc=<0 0 0>, rot=<0 0 0 1>}
 Pose_121 {type=Curve, loc=<0 0 0>, rot=<0 0 0 1>}
 Pose_122 {type=Curve, loc=<0 0 0>, rot=<0 0 0 1>}
 Pose_123 {type=Curve, loc=<0 0 0>, rot=<0 0 0 1>}
 Pose_124 {type=Curve, loc=<0 0 0>, rot=<0 0 0 1>}
 Pose_125 {type=Curve, loc=<0 0 0>, rot=<0 0 0 1>}
 Pose_126 {type=Curve, loc=<0 0 0>, rot=<0 0 0 1>}
 Pose_127 {type=Curve, loc=<0 0 0>, rot=<0 0 0 1>}
 Pose_128 {type=Curve, loc=<0 0 0>, rot=<0 0 0 1>}
 Pose_129 {type=Curve, loc=<0 0 0>, rot=<0 0 0 1>}
 Pose_130 {type=Curve, loc=<0 0 0>, rot=<0 0 0 1>}
 Pose_131 {type=Curve, loc=<0 0 0>, rot=<0 0 0 1>}
 Pose_132 {type=Curve, loc=<0 0 0>, rot=<0 0 0 1>}
 Pose_133 {type=Curve, loc=<0 0 0>, rot=<0 0 0 1>}
 Pose_134 {type=Curve, loc=<0 0 0>, rot=<0 0 0 1>}
 Pose_135 {type=Curve, loc=<0 0 0>, rot=<0 0 0 1>}
 Pose_136 {type=Curve, loc=<0 0 0>, rot=<0 0 0 1>}
 Pose_137 {type=Curve, loc=<0 0 0>, rot=<0 0 0 1>}
 Pose_138 {type=Curve, loc=<0 0 0>, rot=<0 0 0 1>}
 Pose_139 {type=Curve, loc=<0 0 0>, rot=<0 0 0 1>}
 Pose_140 {type=Curve, loc=<0 0 0>, rot=<0 0 0 1>}
 Pose_141 {type=Curve, loc=<0 0 0>, rot=<0 0 0 1>}
 Pose_142 {type=Curve, loc=<0 0 0>, rot=<0 0 0 1>}
 Pose_143 {type=Curve, loc=<0 0 0>, rot=<0 0 0 1>}
 Pose_144 {type=Curve, loc=<0 0 0>, rot=<0 0 0 1>}
 Pose_145 {type=Curve, loc=<0 0 0>, rot=<0 0 0 1>}
 Pose_146 {type=Curve, loc=<0 0 0>, rot=<0 0 0 1>}
 Pose_147 {type=Curve, loc=<0 0 0>, rot=<0 0 0 1>}
 Pose_148 {type=Curve, loc=<0 0 0>, rot=<0 0 0 1>}
 Pose_149 {type=Curve, loc=<0 0 0>, rot=<0 0 0 1>}
 Pose_150 {type=Curve, loc=<0 0 0>, rot=<0 0 0 1>}
 Pose_151 {type=Curve, loc=<0 0 0>, rot=<0 0 0 1>}
 Pose_152 {type=Curve, loc=<0 0 0>, rot=<0 0 0 1>}
 Pose_153 {type=Curve, loc=<0 0 0>, rot=<0 0 0 1>}
 Pose_154 {type=Curve, loc=<0 0 0>, rot=<0 0 0 1>}
 Pose_155 {type=Curve, loc=<0 0 0>, rot=<0 0 0 1>}
 Pose_156 {type=Curve, loc=<0 0 0>, rot=<0 0 0 1>}
 Pose_157 {type=Curve, loc=<0 0 0>, rot=<0 0 0 1>}
 Pose_158 {type=Curve, loc=<0 0 0>, rot=<0 0 0 1>}
 Pose_159 {type=Curve, loc=<0 0 0>, rot=<0 0 0 1>}
 Pose_160 {type=Curve, loc=<0 0 0>, rot=<0 0 0 1>}
 Pose_161 {type=Curve, loc=<0 0 0>, rot=<0 0 0 1>}
 Pose_162 {type=Curve, loc=<0 0 0>, rot=<0 0 0 1>}
 Pose_163 {type=Curve, loc=<0 0 0>, rot=<0 0 0 1>}
 Pose_164 {type=Curve, loc=<0 0 0>, rot=<0 0 0 1>}
 Pose_165 {type=Curve, loc=<0 0 0>, rot=<0 0 0 1>}
 Pose_166 {type=Curve, loc=<0 0 0>, rot=<0 0 0 1>}
 Pose_167 {type=Curve, loc=<0 0 0>, rot=<0 0 0 1>}
 Pose_168 {type=Curve, loc=<0 0 0>, rot=<0 0 0 1>}
 Pose_169 {type=Curve, loc=<0 0 0>, rot=<0 0 0 1>}
 Pose_170 {type=Curve, loc=<0 0 0>, rot=<0 0 0 1>}
 Pose_171 {type=Curve, loc=<0 0 0>, rot=<0 0 0 1>}
 Pose_172 {type=Curve, loc=<0 0 0>, rot=<0 0 0 1>}
 Pose_173 {type=Curve, loc=<0 0 0>, rot=<0 0 0 1>}
 Pose_174 {type=Curve, loc=<0 0 0>, rot=<0 0 0 1>}
 Pose_175 {type=Curve, loc=<0 0 0>, rot=<0 0 0 1>}
 Pose_176 {type=Curve, loc=<0 0 0>, rot=<0 0 0 1>}
 Pose_177 {type=Curve, loc=<0 0 0>, rot=<0 0 0 1>}
 Pose_178 {type=Curve, loc=<0 0 0>, rot=<0 0 0 1>}
 Pose_179 {type=Curve, loc=<0 0 0>, rot=<0 0 0 1>}
 eyeLookDownRight {type=Curve, loc=<0 0 0>, rot=<0 0 0 1>}
 eyeLookOutRight {type=Curve, loc=<0 0 0>, rot=<0 0 0 1>}
 eyeLookUpRight {type=Curve, loc=<0 0 0>, rot=<0 0 0 1>}
 eyeLookOutLeft {type=Curve, loc=<0 0 0>, rot=<0 0 0 1>}
 eyeBlinkLeft {type=Curve, loc=<0 0 0>, rot=<0 0 0 1>}
 eyeBlinkRight {type=Curve, loc=<0 0 0>, rot=<0 0 0 1>}
 eyeSquintLeft {type=Curve, loc=<0 0 0>, rot=<0 0 0 1>}
 eyeSquintRight {type=Curve, loc=<0 0 0>, rot=<0 0 0 1>}
 eyeWideLeft {type=Curve, loc=<0 0 0>, rot=<0 0 0 1>}
 eyeWideRight {type=Curve, loc=<0 0 0>, rot=<0 0 0 1>}
 jawForward {type=Curve, loc=<0 0 0>, rot=<0 0 0 1>}
 jawLeft {type=Curve, loc=<0 0 0>, rot=<0 0 0 1>}
 jawRight {type=Curve, loc=<0 0 0>, rot=<0 0 0 1>}
 jawOpen {type=Curve, loc=<0 0 0>, rot=<0 0 0 1>}
 mouthClose {type=Curve, loc=<0 0 0>, rot=<0 0 0 1>}
 mouthFunnel {type=Curve, loc=<0 0 0>, rot=<0 0 0 1>}
 mouthPucker {type=Curve, loc=<0 0 0>, rot=<0 0 0 1>}
 mouthLeft {type=Curve, loc=<0 0 0>, rot=<0 0 0 1>}
 mouthRight {type=Curve, loc=<0 0 0>, rot=<0 0 0 1>}
 mouthSmileLeft {type=Curve, loc=<0 0 0>, rot=<0 0 0 1>}
 mouthSmileRight {type=Curve, loc=<0 0 0>, rot=<0 0 0 1>}
 mouthFrownLeft {type=Curve, loc=<0 0 0>, rot=<0 0 0 1>}
 mouthFrownRight {type=Curve, loc=<0 0 0>, rot=<0 0 0 1>}
 mouthDimpleLeft {type=Curve, loc=<0 0 0>, rot=<0 0 0 1>}
 mouthDimpleRight {type=Curve, loc=<0 0 0>, rot=<0 0 0 1>}
 mouthStretchLeft {type=Curve, loc=<0 0 0>, rot=<0 0 0 1>}
 mouthStretchRight {type=Curve, loc=<0 0 0>, rot=<0 0 0 1>}
 mouthRollLower {type=Curve, loc=<0 0 0>, rot=<0 0 0 1>}
 mouthRollUpper {type=Curve, loc=<0 0 0>, rot=<0 0 0 1>}
 mouthShrugLower {type=Curve, loc=<0 0 0>, rot=<0 0 0 1>}
 mouthShrugUpper {type=Curve, loc=<0 0 0>, rot=<0 0 0 1>}
 mouthPressLeft {type=Curve, loc=<0 0 0>, rot=<0 0 0 1>}
 mouthPressRight {type=Curve, loc=<0 0 0>, rot=<0 0 0 1>}
 mouthLowerDownLeft {type=Curve, loc=<0 0 0>, rot=<0 0 0 1>}
 mouthLowerDownRight {type=Curve, loc=<0 0 0>, rot=<0 0 0 1>}
 mouthUpperUpLeft {type=Curve, loc=<0 0 0>, rot=<0 0 0 1>}
 mouthUpperUpRight {type=Curve, loc=<0 0 0>, rot=<0 0 0 1>}
 browDownLeft {type=Curve, loc=<0 0 0>, rot=<0 0 0 1>}
 browDownRight {type=Curve, loc=<0 0 0>, rot=<0 0 0 1>}
 browInnerUp {type=Curve, loc=<0 0 0>, rot=<0 0 0 1>}
 browOuterUpLeft {type=Curve, loc=<0 0 0>, rot=<0 0 0 1>}
 browOuterUpRight {type=Curve, loc=<0 0 0>, rot=<0 0 0 1>}
 cheekPuff {type=Curve, loc=<0 0 0>, rot=<0 0 0 1>}
 cheekSquintLeft {type=Curve, loc=<0 0 0>, rot=<0 0 0 1>}
 cheekSquintRight {type=Curve, loc=<0 0 0>, rot=<0 0 0 1>}
 noseSneerLeft {type=Curve, loc=<0 0 0>, rot=<0 0 0 1>}
 noseSneerRight {type=Curve, loc=<0 0 0>, rot=<0 0 0 1>}
 C_glabella_down {type=Curve, loc=<0 0 0>, rot=<0 0 0 1>}
 L_brow_squeeze_in {type=Curve, loc=<0 0 0>, rot=<0 0 0 1>}
 R_brow_squeeze_in {type=Curve, loc=<0 0 0>, rot=<0 0 0 1>}
 L_upper_lip_lower {type=Curve, loc=<0 0 0>, rot=<0 0 0 1>}
 R_upper_lip_lower {type=Curve, loc=<0 0 0>, rot=<0 0 0 1>}
 L_lower_lip_up {type=Curve, loc=<0 0 0>, rot=<0 0 0 1>}
 R_lower_lip_up {type=Curve, loc=<0 0 0>, rot=<0 0 0 1>}
 L_lip_corner_narrow {type=Curve, loc=<0 0 0>, rot=<0 0 0 1>}
 R_lip_corner_narrow {type=Curve, loc=<0 0 0>, rot=<0 0 0 1>}
 phoneme_ch {type=Curve, loc=<0 0 0>, rot=<0 0 0 1>}
 phoneme_mbp {type=Curve, loc=<0 0 0>, rot=<0 0 0 1>}
 phoneme_fv {type=Curve, loc=<0 0 0>, rot=<0 0 0 1>}
 mouth_up {type=Curve, loc=<0 0 0>, rot=<0 0 0 1>}
 mouth_down {type=Curve, loc=<0 0 0>, rot=<0 0 0 1>}
 L_cheek_up {type=Curve, loc=<0 0 0>, rot=<0 0 0 1>}
 R_cheek_up {type=Curve, loc=<0 0 0>, rot=<0 0 0 1>}
 L_upperLidDown {type=Curve, loc=<0 0 0>, rot=<0 0 0 1>}
 R_upperLidDown {type=Curve, loc=<0 0 0>, rot=<0 0 0 1>}
 jaw_up {type=Curve, loc=<0 0 0>, rot=<0 0 0 1>}
 jaw_back {type=Curve, loc=<0 0 0>, rot=<0 0 0 1>}
 tongue_forward {type=Curve, loc=<0 0 0>, rot=<0 0 0 1>}
 tongue_back {type=Curve, loc=<0 0 0>, rot=<0 0 0 1>}
 tongue_up {type=Curve, loc=<0 0 0>, rot=<0 0 0 1>}
 tongue_down {type=Curve, loc=<0 0 0>, rot=<0 0 0 1>}
 basePose {type=Curve, loc=<0 0 0>, rot=<0 0 0 1>}
 Pose_180 {type=Curve, loc=<0 0 0>, rot=<0 0 0 1>}
 Pose_181 {type=Curve, loc=<0 0 0>, rot=<0 0 0 1>}
 Pose_182 {type=Curve, loc=<0 0 0>, rot=<0 0 0 1>}
 Pose_183 {type=Curve, loc=<0 0 0>, rot=<0 0 0 1>}
 Pose_184 {type=Curve, loc=<0 0 0>, rot=<0 0 0 1>}
 Pose_185 {type=Curve, loc=<0 0 0>, rot=<0 0 0 1>}
 Pose_186 {type=Curve, loc=<0 0 0>, rot=<0 0 0 1>}
 Pose_187 {type=Curve, loc=<0 0 0>, rot=<0 0 0 1>}
 Pose_188 {type=Curve, loc=<0 0 0>, rot=<0 0 0 1>}
 Pose_189 {type=Curve, loc=<0 0 0>, rot=<0 0 0 1>}
 Pose_190 {type=Curve, loc=<0 0 0>, rot=<0 0 0 1>}
 Pose_191 {type=Curve, loc=<0 0 0>, rot=<0 0 0 1>}
 Pose_192 {type=Curve, loc=<0 0 0>, rot=<0 0 0 1>}
 Pose_193 {type=Curve, loc=<0 0 0>, rot=<0 0 0 1>}
 Pose_194 {type=Curve, loc=<0 0 0>, rot=<0 0 0 1>}
 Pose_195 {type=Curve, loc=<0 0 0>, rot=<0 0 0 1>}
 Pose_196 {type=Curve, loc=<0 0 0>, rot=<0 0 0 1>}
 Pose_197 {type=Curve, loc=<0 0 0>, rot=<0 0 0 1>}
 Pose_198 {type=Curve, loc=<0 0 0>, rot=<0 0 0 1>}
 Pose_199 {type=Curve, loc=<0 0 0>, rot=<0 0 0 1>}
 Pose_200 {type=Curve, loc=<0 0 0>, rot=<0 0 0 1>}
 Pose_201 {type=Curve, loc=<0 0 0>, rot=<0 0 0 1>}
 Pose_202 {type=Curve, loc=<0 0 0>, rot=<0 0 0 1>}
 Pose_203 {type=Curve, loc=<0 0 0>, rot=<0 0 0 1>}
 Pose_204 {type=Curve, loc=<0 0 0>, rot=<0 0 0 1>}
 Pose_205 {type=Curve, loc=<0 0 0>, rot=<0 0 0 1>}
 Pose_206 {type=Curve, loc=<0 0 0>, rot=<0 0 0 1>}
 Pose_207 {type=Curve, loc=<0 0 0>, rot=<0 0 0 1>}
 Pose_208 {type=Curve, loc=<0 0 0>, rot=<0 0 0 1>}
 Pose_209 {type=Curve, loc=<0 0 0>, rot=<0 0 0 1>}
 Pose_210 {type=Curve, loc=<0 0 0>, rot=<0 0 0 1>}
 Pose_211 {type=Curve, loc=<0 0 0>, rot=<0 0 0 1>}
 Pose_212 {type=Curve, loc=<0 0 0>, rot=<0 0 0 1>}
 Pose_213 {type=Curve, loc=<0 0 0>, rot=<0 0 0 1>}
 Pose_214 {type=Curve, loc=<0 0 0>, rot=<0 0 0 1>}
 Pose_215 {type=Curve, loc=<0 0 0>, rot=<0 0 0 1>}
 Pose_216 {type=Curve, loc=<0 0 0>, rot=<0 0 0 1>}
 Pose_217 {type=Curve, loc=<0 0 0>, rot=<0 0 0 1>}
 Pose_218 {type=Curve, loc=<0 0 0>, rot=<0 0 0 1>}
 Pose_219 {type=Curve, loc=<0 0 0>, rot=<0 0 0 1>}
 Pose_220 {type=Curve, loc=<0 0 0>, rot=<0 0 0 1>}
 Pose_221 {type=Curve, loc=<0 0 0>, rot=<0 0 0 1>}
 Pose_222 {type=Curve, loc=<0 0 0>, rot=<0 0 0 1>}
 Pose_223 {type=Curve, loc=<0 0 0>, rot=<0 0 0 1>}
 Pose_224 {type=Curve, loc=<0 0 0>, rot=<0 0 0 1>}
 Pose_225 {type=Curve, loc=<0 0 0>, rot=<0 0 0 1>}
 Pose_226 {type=Curve, loc=<0 0 0>, rot=<0 0 0 1>}
 Pose_227 {type=Curve, loc=<0 0 0>, rot=<0 0 0 1>}
 Pose_228 {type=Curve, loc=<0 0 0>, rot=<0 0 0 1>}
 Pose_229 {type=Curve, loc=<0 0 0>, rot=<0 0 0 1>}
 Pose_230 {type=Curve, loc=<0 0 0>, rot=<0 0 0 1>}
 Pose_231 {type=Curve, loc=<0 0 0>, rot=<0 0 0 1>}
 Pose_232 {type=Curve, loc=<0 0 0>, rot=<0 0 0 1>}
 Pose_233 {type=Curve, loc=<0 0 0>, rot=<0 0 0 1>}
 Pose_234 {type=Curve, loc=<0 0 0>, rot=<0 0 0 1>}
 Pose_235 {type=Curve, loc=<0 0 0>, rot=<0 0 0 1>}
 Pose_236 {type=Curve, loc=<0 0 0>, rot=<0 0 0 1>}
 Pose_237 {type=Curve, loc=<0 0 0>, rot=<0 0 0 1>}
 Pose_238 {type=Curve, loc=<0 0 0>, rot=<0 0 0 1>}
 Pose_239 {type=Curve, loc=<0 0 0>, rot=<0 0 0 1>}
 Pose_240 {type=Curve, loc=<0 0 0>, rot=<0 0 0 1>}
 Pose_241 {type=Curve, loc=<0 0 0>, rot=<0 0 0 1>}
 Pose_242 {type=Curve, loc=<0 0 0>, rot=<0 0 0 1>}
 Pose_243 {type=Curve, loc=<0 0 0>, rot=<0 0 0 1>}
 Pose_244 {type=Curve, loc=<0 0 0>, rot=<0 0 0 1>}
 Pose_245 {type=Curve, loc=<0 0 0>, rot=<0 0 0 1>}
 Pose_246 {type=Curve, loc=<0 0 0>, rot=<0 0 0 1>}
 Pose_247 {type=Curve, loc=<0 0 0>, rot=<0 0 0 1>}
 Pose_248 {type=Curve, loc=<0 0 0>, rot=<0 0 0 1>}
 Pose_249 {type=Curve, loc=<0 0 0>, rot=<0 0 0 1>}
 Pose_250 {type=Curve, loc=<0 0 0>, rot=<0 0 0 1>}
 Pose_251 {type=Curve, loc=<0 0 0>, rot=<0 0 0 1>}
 Pose_252 {type=Curve, loc=<0 0 0>, rot=<0 0 0 1>}
 Pose_253 {type=Curve, loc=<0 0 0>, rot=<0 0 0 1>}
 Pose_254 {type=Curve, loc=<0 0 0>, rot=<0 0 0 1>}
 Pose_255 {type=Curve, loc=<0 0 0>, rot=<0 0 0 1>}
 Pose_256 {type=Curve, loc=<0 0 0>, rot=<0 0 0 1>}
 Pose_257 {type=Curve, loc=<0 0 0>, rot=<0 0 0 1>}
 Pose_258 {type=Curve, loc=<0 0 0>, rot=<0 0 0 1>}
 neutral {type=Curve, loc=<0 0 0>, rot=<0 0 0 1>}
 Pose_259 {type=Curve, loc=<0 0 0>, rot=<0 0 0 1>}
 cheek_puff_low {type=Curve, loc=<0 0 0>, rot=<0 0 0 1>}
 Pose_260 {type=Curve, loc=<0 0 0>, rot=<0 0 0 1>}
 Pose_261 {type=Curve, loc=<0 0 0>, rot=<0 0 0 1>}
 Pose_262 {type=Curve, loc=<0 0 0>, rot=<0 0 0 1>}
 Pose_263 {type=Curve, loc=<0 0 0>, rot=<0 0 0 1>}
 Pose_264 {type=Curve, loc=<0 0 0>, rot=<0 0 0 1>}
 Pose_265 {type=Curve, loc=<0 0 0>, rot=<0 0 0 1>}
 Pose_266 {type=Curve, loc=<0 0 0>, rot=<0 0 0 1>}
 Pose_267 {type=Curve, loc=<0 0 0>, rot=<0 0 0 1>}
 Pose_268 {type=Curve, loc=<0 0 0>, rot=<0 0 0 1>}
 Pose_269 {type=Curve, loc=<0 0 0>, rot=<0 0 0 1>}
 Pose_270 {type=Curve, loc=<0 0 0>, rot=<0 0 0 1>}
 Pose_271 {type=Curve, loc=<0 0 0>, rot=<0 0 0 1>}
 Pose_272 {type=Curve, loc=<0 0 0>, rot=<0 0 0 1>}
 Pose_273 {type=Curve, loc=<0 0 0>, rot=<0 0 0 1>}
 Pose_274 {type=Curve, loc=<0 0 0>, rot=<0 0 0 1>}
 Pose_275 {type=Curve, loc=<0 0 0>, rot=<0 0 0 1>}
 Pose_276 {type=Curve, loc=<0 0 0>, rot=<0 0 0 1>}
 Pose_277 {type=Curve, loc=<0 0 0>, rot=<0 0 0 1>}
 Pose_278 {type=Curve, loc=<0 0 0>, rot=<0 0 0 1>}
 Pose_279 {type=Curve, loc=<0 0 0>, rot=<0 0 0 1>}
 Pose_280 {type=Curve, loc=<0 0 0>, rot=<0 0 0 1>}
 Pose_281 {type=Curve, loc=<0 0 0>, rot=<0 0 0 1>}
 Pose_282 {type=Curve, loc=<0 0 0>, rot=<0 0 0 1>}
 Pose_283 {type=Curve, loc=<0 0 0>, rot=<0 0 0 1>}
 Pose_284 {type=Curve, loc=<0 0 0>, rot=<0 0 0 1>}
 Pose_285 {type=Curve, loc=<0 0 0>, rot=<0 0 0 1>}
 Pose_286 {type=Curve, loc=<0 0 0>, rot=<0 0 0 1>}
 Pose_287 {type=Curve, loc=<0 0 0>, rot=<0 0 0 1>}
 Pose_288 {type=Curve, loc=<0 0 0>, rot=<0 0 0 1>}
 Pose_289 {type=Curve, loc=<0 0 0>, rot=<0 0 0 1>}
 Pose_290 {type=Curve, loc=<0 0 0>, rot=<0 0 0 1>}
 Pose_291 {type=Curve, loc=<0 0 0>, rot=<0 0 0 1>}
 Pose_292 {type=Curve, loc=<0 0 0>, rot=<0 0 0 1>}
 Pose_293 {type=Curve, loc=<0 0 0>, rot=<0 0 0 1>}
 Pose_294 {type=Curve, loc=<0 0 0>, rot=<0 0 0 1>}
 Pose_295 {type=Curve, loc=<0 0 0>, rot=<0 0 0 1>}
 Pose_296 {type=Curve, loc=<0 0 0>, rot=<0 0 0 1>}
 Pose_297 {type=Curve, loc=<0 0 0>, rot=<0 0 0 1>}
 Pose_298 {type=Curve, loc=<0 0 0>, rot=<0 0 0 1>}
 Pose_299 {type=Curve, loc=<0 0 0>, rot=<0 0 0 1>}
 Pose_300 {type=Curve, loc=<0 0 0>, rot=<0 0 0 1>}
 Pose_301 {type=Curve, loc=<0 0 0>, rot=<0 0 0 1>}
 Pose_302 {type=Curve, loc=<0 0 0>, rot=<0 0 0 1>}
 Pose_303 {type=Curve, loc=<0 0 0>, rot=<0 0 0 1>}
 Pose_304 {type=Curve, loc=<0 0 0>, rot=<0 0 0 1>}
 Pose_305 {type=Curve, loc=<0 0 0>, rot=<0 0 0 1>}
 Pose_306 {type=Curve, loc=<0 0 0>, rot=<0 0 0 1>}
 Pose_307 {type=Curve, loc=<0 0 0>, rot=<0 0 0 1>}
 Pose_308 {type=Curve, loc=<0 0 0>, rot=<0 0 0 1>}
 Pose_309 {type=Curve, loc=<0 0 0>, rot=<0 0 0 1>}
 Pose_310 {type=Curve, loc=<0 0 0>, rot=<0 0 0 1>}
 Pose_311 {type=Curve, loc=<0 0 0>, rot=<0 0 0 1>}
 Pose_312 {type=Curve, loc=<0 0 0>, rot=<0 0 0 1>}
 Pose_313 {type=Curve, loc=<0 0 0>, rot=<0 0 0 1>}
 Pose_314 {type=Curve, loc=<0 0 0>, rot=<0 0 0 1>}
 Pose_315 {type=Curve, loc=<0 0 0>, rot=<0 0 0 1>}
 Pose_316 {type=Curve, loc=<0 0 0>, rot=<0 0 0 1>}
 Pose_317 {type=Curve, loc=<0 0 0>, rot=<0 0 0 1>}
 Pose_318 {type=Curve, loc=<0 0 0>, rot=<0 0 0 1>}
 Pose_319 {type=Curve, loc=<0 0 0>, rot=<0 0 0 1>}
 Pose_320 {type=Curve, loc=<0 0 0>, rot=<0 0 0 1>}
 Pose_321 {type=Curve, loc=<0 0 0>, rot=<0 0 0 1>}
 Desert_Island_Jonesy_Mouth {type=Curve, loc=<0 0 0>, rot=<0 0 0 1>}
 Pose_322 {type=Curve, loc=<0 0 0>, rot=<0 0 0 1>}
 HT_L1_TransitionCurve {type=Curve, loc=<0 0 0>, rot=<0 0 0 1>}
 Hair_CapShape {type=Curve, loc=<0 0 0>, rot=<0 0 0 1>}
 Hair_CapShape1 {type=Curve, loc=<0 0 0>, rot=<0 0 0 1>}
 jekyll {type=Curve, loc=<0 0 0>, rot=<0 0 0 1>}
 JH_TransformCurve {type=Curve, loc=<0 0 0>, rot=<0 0 0 1>}
 eyes_LOD0 {type=Curve, loc=<0 0 0>, rot=<0 0 0 1>}
 MaterialCurveA {type=Curve, loc=<0 0 0>, rot=<0 0 0 1>}
 MaterialCurveB {type=Curve, loc=<0 0 0>, rot=<0 0 0 1>}
 MaterialCurveC {type=Curve, loc=<0 0 0>, rot=<0 0 0 1>}
 MaterialCurveD {type=Curve, loc=<0 0 0>, rot=<0 0 0 1>}
 MaterialCurveE {type=Curve, loc=<0 0 0>, rot=<0 0 0 1>}
 Custom_Frontend_Face_Idle {type=Curve, loc=<0 0 0>, rot=<0 0 0 1>}
 blendParent1 {type=Curve, loc=<0 0 0>, rot=<0 0 0 1>}
 BaseHeadDefaultFaceIdle {type=Curve, loc=<0 0 0>, rot=<0 0 0 1>}
 VeiledTransitionGlowAmount {type=Curve, loc=<0 0 0>, rot=<0 0 0 1>}
 is_3L {type=Curve, loc=<0 0 0>, rot=<0 0 0 1>}
 Hair_Hat_LOD0 {type=Curve, loc=<0 0 0>, rot=<0 0 0 1>}
 Hair_Helmet_LOD0 {type=Curve, loc=<0 0 0>, rot=<0 0 0 1>}
 Hair_Null_LOD0 {type=Curve, loc=<0 0 0>, rot=<0 0 0 1>}
 Pose_323 {type=Curve, loc=<0 0 0>, rot=<0 0 0 1>}
 Pose_324 {type=Curve, loc=<0 0 0>, rot=<0 0 0 1>}
 Pose_325 {type=Curve, loc=<0 0 0>, rot=<0 0 0 1>}
 Pose_326 {type=Curve, loc=<0 0 0>, rot=<0 0 0 1>}
 Pose_327 {type=Curve, loc=<0 0 0>, rot=<0 0 0 1>}
 Pose_328 {type=Curve, loc=<0 0 0>, rot=<0 0 0 1>}
 Pose_329 {type=Curve, loc=<0 0 0>, rot=<0 0 0 1>}
 Pose_330 {type=Curve, loc=<0 0 0>, rot=<0 0 0 1>}
 Pose_331 {type=Curve, loc=<0 0 0>, rot=<0 0 0 1>}
 Pose_332 {type=Curve, loc=<0 0 0>, rot=<0 0 0 1>}
 Pose_333 {type=Curve, loc=<0 0 0>, rot=<0 0 0 1>}
 Pose_334 {type=Curve, loc=<0 0 0>, rot=<0 0 0 1>}
 Pose_335 {type=Curve, loc=<0 0 0>, rot=<0 0 0 1>}
 Pose_336 {type=Curve, loc=<0 0 0>, rot=<0 0 0 1>}
 Pose_337 {type=Curve, loc=<0 0 0>, rot=<0 0 0 1>}
 Pose_338 {type=Curve, loc=<0 0 0>, rot=<0 0 0 1>}
 Pose_339 {type=Curve, loc=<0 0 0>, rot=<0 0 0 1>}
 Pose_340 {type=Curve, loc=<0 0 0>, rot=<0 0 0 1>}
 Pose_341 {type=Curve, loc=<0 0 0>, rot=<0 0 0 1>}
 Pose_342 {type=Curve, loc=<0 0 0>, rot=<0 0 0 1>}
 Pose_343 {type=Curve, loc=<0 0 0>, rot=<0 0 0 1>}
 Pose_344 {type=Curve, loc=<0 0 0>, rot=<0 0 0 1>}
 Pose_345 {type=Curve, loc=<0 0 0>, rot=<0 0 0 1>}
 Pose_346 {type=Curve, loc=<0 0 0>, rot=<0 0 0 1>}
 Pose_347 {type=Curve, loc=<0 0 0>, rot=<0 0 0 1>}
 Pose_348 {type=Curve, loc=<0 0 0>, rot=<0 0 0 1>}
 Pose_349 {type=Curve, loc=<0 0 0>, rot=<0 0 0 1>}
 Pose_350 {type=Curve, loc=<0 0 0>, rot=<0 0 0 1>}
 Pose_351 {type=Curve, loc=<0 0 0>, rot=<0 0 0 1>}
 Pose_352 {type=Curve, loc=<0 0 0>, rot=<0 0 0 1>}
 Pose_353 {type=Curve, loc=<0 0 0>, rot=<0 0 0 1>}
 Pose_354 {type=Curve, loc=<0 0 0>, rot=<0 0 0 1>}
 Pose_355 {type=Curve, loc=<0 0 0>, rot=<0 0 0 1>}
 Pose_356 {type=Curve, loc=<0 0 0>, rot=<0 0 0 1>}
 Pose_357 {type=Curve, loc=<0 0 0>, rot=<0 0 0 1>}
 Pose_358 {type=Curve, loc=<0 0 0>, rot=<0 0 0 1>}
 Pose_359 {type=Curve, loc=<0 0 0>, rot=<0 0 0 1>}
 Pose_360 {type=Curve, loc=<0 0 0>, rot=<0 0 0 1>}
 Pose_361 {type=Curve, loc=<0 0 0>, rot=<0 0 0 1>}
 Pose_362 {type=Curve, loc=<0 0 0>, rot=<0 0 0 1>}
 Pose_363 {type=Curve, loc=<0 0 0>, rot=<0 0 0 1>}
 Pose_364 {type=Curve, loc=<0 0 0>, rot=<0 0 0 1>}
 Pose_365 {type=Curve, loc=<0 0 0>, rot=<0 0 0 1>}
 Pose_366 {type=Curve, loc=<0 0 0>, rot=<0 0 0 1>}
 Pose_367 {type=Curve, loc=<0 0 0>, rot=<0 0 0 1>}
 Pose_368 {type=Curve, loc=<0 0 0>, rot=<0 0 0 1>}
 Pose_369 {type=Curve, loc=<0 0 0>, rot=<0 0 0 1>}
 Pose_370 {type=Curve, loc=<0 0 0>, rot=<0 0 0 1>}
 Pose_371 {type=Curve, loc=<0 0 0>, rot=<0 0 0 1>}
 Pose_372 {type=Curve, loc=<0 0 0>, rot=<0 0 0 1>}
 Pose_373 {type=Curve, loc=<0 0 0>, rot=<0 0 0 1>}
 Pose_374 {type=Curve, loc=<0 0 0>, rot=<0 0 0 1>}
 Pose_375 {type=Curve, loc=<0 0 0>, rot=<0 0 0 1>}
 Pose_376 {type=Curve, loc=<0 0 0>, rot=<0 0 0 1>}
 Pose_377 {type=Curve, loc=<0 0 0>, rot=<0 0 0 1>}
 Pose_378 {type=Curve, loc=<0 0 0>, rot=<0 0 0 1>}
 Pose_379 {type=Curve, loc=<0 0 0>, rot=<0 0 0 1>}
 Pose_380 {type=Curve, loc=<0 0 0>, rot=<0 0 0 1>}
 Pose_381 {type=Curve, loc=<0 0 0>, rot=<0 0 0 1>}
 Pose_382 {type=Curve, loc=<0 0 0>, rot=<0 0 0 1>}
 Pose_383 {type=Curve, loc=<0 0 0>, rot=<0 0 0 1>}
 Pose_384 {type=Curve, loc=<0 0 0>, rot=<0 0 0 1>}
 Pose_385 {type=Curve, loc=<0 0 0>, rot=<0 0 0 1>}
 Pose_386 {type=Curve, loc=<0 0 0>, rot=<0 0 0 1>}
 Pose_387 {type=Curve, loc=<0 0 0>, rot=<0 0 0 1>}
 Pose_388 {type=Curve, loc=<0 0 0>, rot=<0 0 0 1>}
 Pose_389 {type=Curve, loc=<0 0 0>, rot=<0 0 0 1>}
 Pose_390 {type=Curve, loc=<0 0 0>, rot=<0 0 0 1>}
 Pose_391 {type=Curve, loc=<0 0 0>, rot=<0 0 0 1>}
 Pose_392 {type=Curve, loc=<0 0 0>, rot=<0 0 0 1>}
 Pose_393 {type=Curve, loc=<0 0 0>, rot=<0 0 0 1>}
 Pose_394 {type=Curve, loc=<0 0 0>, rot=<0 0 0 1>}
 Pose_395 {type=Curve, loc=<0 0 0>, rot=<0 0 0 1>}
 Pose_396 {type=Curve, loc=<0 0 0>, rot=<0 0 0 1>}
 Pose_397 {type=Curve, loc=<0 0 0>, rot=<0 0 0 1>}
 Pose_398 {type=Curve, loc=<0 0 0>, rot=<0 0 0 1>}
 Pose_399 {type=Curve, loc=<0 0 0>, rot=<0 0 0 1>}
 Pose_400 {type=Curve, loc=<0 0 0>, rot=<0 0 0 1>}
 Pose_401 {type=Curve, loc=<0 0 0>, rot=<0 0 0 1>}
 Pose_402 {type=Curve, loc=<0 0 0>, rot=<0 0 0 1>}
 Pose_403 {type=Curve, loc=<0 0 0>, rot=<0 0 0 1>}
 Pose_404 {type=Curve, loc=<0 0 0>, rot=<0 0 0 1>}
 Pose_405 {type=Curve, loc=<0 0 0>, rot=<0 0 0 1>}
 Pose_406 {type=Curve, loc=<0 0 0>, rot=<0 0 0 1>}
 Pose_407 {type=Curve, loc=<0 0 0>, rot=<0 0 0 1>}
 Pose_408 {type=Curve, loc=<0 0 0>, rot=<0 0 0 1>}
 Pose_409 {type=Curve, loc=<0 0 0>, rot=<0 0 0 1>}
 Pose_410 {type=Curve, loc=<0 0 0>, rot=<0 0 0 1>}
 Pose_411 {type=Curve, loc=<0 0 0>, rot=<0 0 0 1>}
 Pose_412 {type=Curve, loc=<0 0 0>, rot=<0 0 0 1>}
 Pose_413 {type=Curve, loc=<0 0 0>, rot=<0 0 0 1>}
 Pose_414 {type=Curve, loc=<0 0 0>, rot=<0 0 0 1>}
 mouth_close_pose {type=Curve, loc=<0 0 0>, rot=<0 0 0 1>}
 mouth_pucker_pose {type=Curve, loc=<0 0 0>, rot=<0 0 0 1>}
 mouth_left_pose {type=Curve, loc=<0 0 0>, rot=<0 0 0 1>}
 mouth_right_pose {type=Curve, loc=<0 0 0>, rot=<0 0 0 1>}
 L_mouth_stretch_pose {type=Curve, loc=<0 0 0>, rot=<0 0 0 1>}
 R_mouth_stretch_pose {type=Curve, loc=<0 0 0>, rot=<0 0 0 1>}
 mouth_roll_lower_pose {type=Curve, loc=<0 0 0>, rot=<0 0 0 1>}
 mouth_roll_upper_pose {type=Curve, loc=<0 0 0>, rot=<0 0 0 1>}
 mouth_shrug_lower_pose {type=Curve, loc=<0 0 0>, rot=<0 0 0 1>}
 mouth_shrug_upper_pose {type=Curve, loc=<0 0 0>, rot=<0 0 0 1>}
 L_mouth_press_pose {type=Curve, loc=<0 0 0>, rot=<0 0 0 1>}
 R_mouth_press_pose {type=Curve, loc=<0 0 0>, rot=<0 0 0 1>}
 cheek_puff_pose {type=Curve, loc=<0 0 0>, rot=<0 0 0 1>}
 L_nose_wrinkler_pose {type=Curve, loc=<0 0 0>, rot=<0 0 0 1>}
 R_nose_wrinkler_pose {type=Curve, loc=<0 0 0>, rot=<0 0 0 1>}
]
