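# Physics asset for the SirWolf medium male player head, providing capsule bodies for five dynamic hair bones, the head and upper neck.
# Each hair bone constrains directly to the head rather than forming a chain, with zero twist limit and tighter swing limits on bones 03 to 05.
# The head joins neck_02 with free swing and twist, and its capsule has zero size, leaving the neck capsule as the sized non-hair shape.
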PhysicsAsset /Game/Characters/Player/Male/Medium/Heads/M_MED_SirWolf_Head/Meshes/M_MED_SirWolf_Head_Physics
bodies {
 dyn_hair_01 [{shapeType=Capsule, length=3, radius=3, shapeName=dyn_hair_01_capsule, center=<2 0 0>, rotation=<rot 90 0 0>}]
 dyn_hair_02 [{shapeType=Capsule, length=3, radius=3, shapeName=dyn_hair_02_capsule, center=<-2 0 0>, rotation=<rot 90 0 0>}]
 dyn_hair_03 [{shapeType=Capsule, length=3, radius=3, shapeName=dyn_hair_01_capsule, center=<1 0 0>, rotation=<rot 90 0 0>}]
 dyn_hair_04 [{shapeType=Capsule, length=3, radius=3, shapeName=dyn_hair_01_capsule, center=<1 0 0>, rotation=<rot 90 0 0>}]
 dyn_hair_05 [{shapeType=Capsule, length=3, radius=3, shapeName=dyn_hair_01_capsule, center=<1 0 0>, rotation=<rot 90 0 0>}]
 head [{shapeType=Capsule, length=0, radius=0, shapeName=head_capsule, center=<5.9941 6.6388 0.1159>, rotation=<rot 0 -140.4026 -87.485>}]
 neck_02 [{
   shapeType=Capsule
   length=9.1666
   radius=2.0145
   shapeName=neck_02_capsule
   center=<6.0578 5.9055 -0.2159>
   rotation=<rot 2.4059 -90 -2.8539>
  }]
}
constraints [
 {bone1Name=dyn_hair_01, bone2Name=head, swing1LimitDegrees=30, swing2LimitDegrees=30, twistLimitDegrees=0}
 {bone1Name=dyn_hair_02, bone2Name=head, swing1LimitDegrees=30, swing2LimitDegrees=30, twistLimitDegrees=0}
 {
  bone1Name=head
  bone2Name=neck_02
  swing1LimitDegrees=45
  swing1Motion=Free
  swing2LimitDegrees=45
  swing2Motion=Free
  twistLimitDegrees=45
  twistMotion=Free
 }
 {bone1Name=dyn_hair_04, bone2Name=head, swing1LimitDegrees=10, swing2LimitDegrees=10, twistLimitDegrees=0}
 {bone1Name=dyn_hair_05, bone2Name=head, swing1LimitDegrees=10, swing2LimitDegrees=10, twistLimitDegrees=0}
 {bone1Name=dyn_hair_03, bone2Name=head, swing1LimitDegrees=10, swing2LimitDegrees=10, twistLimitDegrees=0}
]
